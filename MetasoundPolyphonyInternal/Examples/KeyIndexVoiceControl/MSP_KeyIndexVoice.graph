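# Example MetaSound patch for a mono polyphony voice, implementing the Voice, Gain and KeyIndex control interfaces with a sine tone.
# The key index plus one multiplies the base frequency input, so successive indices step up the harmonic series.
# Start and Stop trigger attack and release of an ADSR envelope, which shapes the sine before gain scaling.
# Envelope completion fires OnFinished, while both priority pins are left unwired.
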MetaSoundPatch /MetasoundPolyphonyInternal/Examples/KeyIndexVoiceControl/MSP_KeyIndexVoice
inputs [
 Polyphony.Voice.Control.Gain.Gain {type=Float}
 Polyphony.Voice.Control.KeyIndex.Index {type=Int32}
 Polyphony.Voice.Control.Sine.Frequency {type=Float}
 Polyphony.Voice.In Priority {type=Int32}
 Polyphony.Voice.Start {type=Trigger}
 Polyphony.Voice.Stop {type=Trigger}
]
interfaces [Polyphony.Voice.Control.Gain:1.0, Polyphony.Voice.Control.KeyIndex:1.0, Polyphony.Voice:1.0, UE.OutputFormat.Mono:1.0]
outputs [
 Polyphony.Voice.OnFinished {type=Trigger}
 Polyphony.Voice.Out Priority {type=Int32}
 "UE.OutputFormat.Mono.Audio:0" {type=Audio}
]
pages [Default {
  input_defaults {
   Polyphony.Voice.Control.Gain.Gain=0.5
   Polyphony.Voice.Control.KeyIndex.Index=null
   Polyphony.Voice.Control.Sine.Frequency=440
   Polyphony.Voice.In Priority=null
   Polyphony.Voice.Start=null
   Polyphony.Voice.Stop=null
  }
  nodes [
   ADSREnvelopeADSREnvelope_0 = ADSR Envelope.ADSR Envelope:Audio(Trigger Attack=^Polyphony.Voice.Start, Trigger Release=^Polyphony.Voice.Stop) {also {defaults {Hard Reset="<MetaSoundBuilderResult.SUCCEEDED: 0>"}}}
   n1 = UE.Multiply:Float(
    AdditionalOperands=^Polyphony.Voice.Control.Sine.Frequency
    PrimaryOperand=Convert.Float:Int32(Int32=UE.Add:Int32(PrimaryOperand=^Polyphony.Voice.Control.KeyIndex.Index) {also {defaults {AdditionalOperands=1}}}.Out).Float
   )
   Multiply_1 = UE.Multiply:Audio by Float(
    AdditionalOperands=^Polyphony.Voice.Control.Gain.Gain
    PrimaryOperand=UE.Multiply:Audio(AdditionalOperands=@ADSREnvelopeADSREnvelope_0.Out Envelope, PrimaryOperand=UE.Sine:Audio(Frequency=@n1.Out).Audio).Out
   )
  ]
  links [{dst="@Polyphony.Voice.OnFinished", src=ADSREnvelopeADSREnvelope_0.On Done}, {dst="@UE.OutputFormat.Mono.Audio:0", src=Multiply_1.Out}]
 }]
