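# Mono MetaSound Source example that demonstrates polyphonic voice control by key index, starting and stopping voices on a fixed timeline from On Play.
# Trigger Delays feed two Trigger Routes that produce staggered starts for keys 0, 1 and 2, then stops in reverse order.
# A merged route supplies the key index, while a Trigger Select splits events into Trigger or Stop on the TriggerIndexVoiceController.
# Its control stream drives a PolyphonyPatchPlayer running MSP_KeyIndexVoice.
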MetaSoundSource /MetasoundPolyphonyInternal/Examples/KeyIndexVoiceControl/MS_Polyphony_KeyIndexControl
inputs [UE.Source.OnPlay {type=Trigger}]
interfaces [UE.OutputFormat.Mono:1.0, UE.Source:1.1]
outputs ["UE.OutputFormat.Mono.Audio:0" {type=Audio}]
pages [Default {
  input_defaults {UE.Source.OnPlay=null}
  nodes [
   TriggerDelay_1 = UE.Trigger Delay(In=^UE.Source.OnPlay) {also {defaults {Delay Time=2}}}
   TriggerDelay_4 = UE.Trigger Delay(In=^UE.Source.OnPlay)
   TriggerRouteTriggerRouteInt323_0 = "TriggerRoute.Trigger Route (Int32, 3):Int32"(Set 0=^UE.Source.OnPlay, Set 2=@TriggerDelay_1.Out, Set 1=@TriggerDelay_4.Out) {also {defaults {Value 1=1, Value 2=2}}}
   TriggerDelay_0 = UE.Trigger Delay(In=^UE.Source.OnPlay) {also {defaults {Delay Time=5}}}
   TriggerDelay_2 = UE.Trigger Delay(In=^UE.Source.OnPlay) {also {defaults {Delay Time=3}}}
   TriggerDelay_3 = UE.Trigger Delay(In=^UE.Source.OnPlay) {also {defaults {Delay Time=4}}}
   TriggerRouteTriggerRouteInt323_1 = "TriggerRoute.Trigger Route (Int32, 3):Int32"(Set 2=@TriggerDelay_0.Out, Set 0=@TriggerDelay_2.Out, Set 1=@TriggerDelay_3.Out) {also {defaults {Value 0=2, Value 1=1, Value 2=0}}}
   TriggerRouteTriggerRouteInt322_1 = "TriggerRoute.Trigger Route (Int32, 2):Int32"(
    Set 0=@TriggerRouteTriggerRouteInt323_0.On Set
    Value 0=@TriggerRouteTriggerRouteInt323_0.Value
    Set 1=@TriggerRouteTriggerRouteInt323_1.On Set
    Value 1=@TriggerRouteTriggerRouteInt323_1.Value
   )
   n1 = UE.Reroute(Value=@TriggerRouteTriggerRouteInt322_1.Value)
   TriggerRouteTriggerRouteInt322_0 = "TriggerRoute.Trigger Route (Int32, 2):Int32"(Set 0=@TriggerRouteTriggerRouteInt323_0.On Set, Set 1=@TriggerRouteTriggerRouteInt323_1.On Set) {also {defaults {Value 0=0, Value 1=1}}}
   TriggerSelect2_0 = "UE.Trigger Select (2)"(Index=@TriggerRouteTriggerRouteInt322_0.Value, In=@TriggerRouteTriggerRouteInt322_1.On Set)
   MetasoundPolyphonyInternalPolyphonyPatchPlayer_0 = MetasoundPolyphonyInternal.PolyphonyPatchPlayer:Mono(Voice Control Stream=MetasoundPolyphonyInternal.TriggerIndexVoiceController(Key Index=UE.Reroute(Value=@n1.Value).Value, Trigger=@TriggerSelect2_0.Out 0, Stop=@TriggerSelect2_0.Out 1) {also {version [0,1]}}.Voice Control Stream) {also {defaults {MetaSound Asset=&/MSP_KeyIndexVoice}, version [0,1]}}
  ]
  links [{dst="@UE.OutputFormat.Mono.Audio:0", src=MetasoundPolyphonyInternalPolyphonyPatchPlayer_0.Out Mono}]
 }]
properties {
 assetImportData=&self:AssetImportData
 bLooping=true
 duration=10000
 sampleRate=48000
}
